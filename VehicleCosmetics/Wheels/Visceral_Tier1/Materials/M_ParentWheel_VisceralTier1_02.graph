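# Masked parent material for the Visceral Tier 1 vehicle wheel, using vertex red to blend between rim and tire shading.
# The rim fakes reflections with a reflection-vector swirl texture and adds tinted, mask-driven teeth.
# The tire layers detail normals into view-angle shading, plus a color-parameterized tread glow distorted by panning noise.
# The Lods_Alpha static switch toggles rim mask detail and alpha cutout.
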Material /VehicleCosmetics/Wheels/Visceral_Tier1/Materials/M_ParentWheel_VisceralTier1_02
graph {
 properties {
  bAutomaticallySetUsageInEditor=false
  bUsedWithSkeletalMesh=true
  blendMode=BLEND_Masked
  previewMesh=/VehicleCosmetics/Wheels/Visceral_Tier1/Mesh/SM_Visceral_Tier_LODs
  usedShadingModels=Default Lit
 }
 nodes [
  Constant()
  LightVector()
  n1 = Constant()
  n2 = Panner() {speedX=0.2}
  n3 = Add(
   A=Multiply(
    A=TextureSample(UVs=@n2) {samplerSource=SSM_Wrap_WorldGroupSettings, texture=/VehicleCosmetics/Wheels/_SharedTextures/FX/CloudDistort_T}.G
    B=Constant() {r=0.05}
   )
   B=TextureCoordinate()
  )
  n4 = TextureSample() {samplerSource=SSM_Wrap_WorldGroupSettings, texture=/VehicleCosmetics/Wheels/Visceral_Tier1/Textures/T_Visceral_Tier1_Tire_M}
  n5 = Multiply(
   A=ConstantBiasScale(None=TextureSample(UVs=@n3) {samplerSource=SSM_Wrap_WorldGroupSettings, texture=/VehicleCosmetics/Wheels/Visceral_Tier1/Textures/Visceral_diffuse_grayscale}.RGB)
   B=@n4.G
  )
  n6 = ScalarParameter() {defaultValue=3, parameterName=TreadGlowAmount}
  n7 = VectorParameter() {parameterName=CustomColor_Tire, defaultValue=<rgba 1 0.03 0.02 1>}
  n8 = ScalarParameter() {defaultValue=3, parameterName=TreadGlowBrightness}
  n9 = VertexColor()
  LinearInterpolate(A=@n1, B=Multiply(A=Power(Base=@n5, Exp=@n6), B=Multiply(A=@n7.RGB, B=@n8)), Alpha=@n9.R)
  n10 = TextureSample() {samplerType=SAMPLERTYPE_LinearColor, texture=/VehicleCosmetics/Wheels/Visceral_Tier1/Textures/T_Visceral_Tier1_Rim_M_LOD}
  n11 = Add(A=@n10.R, B=@n10.R)
  n12 = NamedRerouteUsage()
  LinearInterpolate(
   A=Subtract(A=ScalarParameter() {defaultValue=1, parameterName=Metallic_Rim}, B=@n11)
   B=LinearInterpolate(A=Constant(), B=ScalarParameter() {parameterName=Metallic_Tier}, Alpha=OneMinus(None=@n12))
   Alpha=NamedRerouteUsage()
  )
  n13 = Constant2Vector() {g=0.05, r=0.05}
  n14 = ComponentMask(None=ReflectionVectorWS()) {g=true, r=true}
  n15 = Add(
   A=TextureSample(UVs=Multiply(A=@n13, B=@n14)) {samplerSource=SSM_Wrap_WorldGroupSettings, texture=/VehicleCosmetics/Wheels/_SharedTextures/FX/Swirls_D}.RGB
   B=ScalarParameter() {defaultValue=0.3, parameterName=ReflectionBrightness}
  )
  LinearInterpolate(A=@n15, B=Constant() {r=1}, Alpha=@n11)
  n16 = TextureSampleParameter2D() {
   parameterName=BC
   samplerSource=SSM_Wrap_WorldGroupSettings
   texture=/VehicleCosmetics/Wheels/Visceral_Tier1/Textures/T_Visceral_Tier1_Rim_BC
  }
  n17 = Multiply(A=@n15, B=@n16.RGB)
  n18 = StaticSwitchParameter(True=LinearInterpolate(A=@n17, B=@n16.RGB, Alpha=@n11), False=@n17) {defaultValue=true, parameterName=Lods_Alpha}
  n19 = VectorParameter() {parameterName=TeethColor1, defaultValue=<rgba 1 1 1 1>}
  n20 = Multiply(
   A=Desaturation(None=Multiply(A=@n19.RGB, B=@n16.RGB), Fraction=ScalarParameter() {defaultValue=0.2, parameterName=Desturation}) {luminanceFactors=<rgba 0.3 0.59 0.11 1>}
   B=ScalarParameter() {defaultValue=1.7, parameterName=Teeth01_Glow}
  )
  n21 = LinearInterpolate(
   A=TextureSample() {samplerSource=SSM_Wrap_WorldGroupSettings, texture=/VehicleCosmetics/Wheels/Visceral_Tier1/Textures/Visceral_tread}.RGB
   B=@n1
   Alpha=@n4.G
  )
  n22 = TextureSample() {
   samplerSource=SSM_Wrap_WorldGroupSettings
   samplerType=SAMPLERTYPE_Normal
   texture=/VehicleCosmetics/Wheels/Visceral_Tier1/Textures/T_Visceral_Tier1_Tire_N
  }
  n23 = Constant3Vector() {constant=<rgba 0.2 0.2 1 0>}
  n24 = TextureSample(UVs=TextureCoordinate() {uTiling=2, vTiling=2}) {
   samplerSource=SSM_Wrap_WorldGroupSettings
   samplerType=SAMPLERTYPE_Normal
   texture=/VehicleCosmetics/Wheels/Onetrue_Negative/Textures/Matte_N
  }
  n25 = Constant3Vector() {constant=<rgba 0.375 0.375 0 0>}
  n26 = LinearInterpolate(A=Add(A=Multiply(A=@n22.RGB, B=@n23), B=Multiply(A=@n24.RGB, B=@n25)), B=Constant3Vector() {constant=<rgba 0 0 1 0>}, Alpha=@n4.G)
  n27 = DotProduct(A=@n26, B=CameraVectorWS())
  n28 = Constant() {r=6}
  n29 = Multiply(A=Multiply(A=@n21, B=Constant() {r=16}), B=ConstantBiasScale(None=Power(Base=@n27, Exp=@n28)) {bias=0.0156})
  n30 = Clamp(None=DotProduct(A=@n26, B=Constant() {r=0.6}))
  n31 = Constant() {r=2}
  n32 = OneMinus(None=@n27)
  n33 = Multiply(A=Clamp(None=Multiply(A=@n31, B=@n27)) {minDefault=0.35}, B=Clamp(None=ConstantBiasScale(None=@n32) {bias=-0.375, scale=1.666}))
  n34 = Multiply(A=@n33, B=@n33)
  LinearInterpolate(
   A=StaticSwitchParameter(True=LinearInterpolate(A=@n18, B=@n20, Alpha=@n10.R), False=@n17) {defaultValue=true, parameterName=Lods_Alpha}
   B=Clamp(None=LinearInterpolate(A=@n29, B=@n30, Alpha=@n34))
   Alpha=@n9.R
  )
  n35 = Multiply(
   A=ScalarParameter() {defaultValue=0.6, parameterName=Roughness_Rim}
   B=Multiply(A=ConstantBiasScale(None=@n16.R), B=ScalarParameter() {defaultValue=0.1, parameterName=RoughIntensity})
  )
  n36 = LinearInterpolate(A=@n35, B=ScalarParameter() {defaultValue=0.5, parameterName=Teeth_Roughness_01}, Alpha=@n10.R)
  n37 = ScalarParameter() {defaultValue=0.45, parameterName=Teeth_Roughness_02}
  n38 = Multiply(A=@n30, B=@n21)
  n39 = Constant() {r=20}
  n40 = Multiply(A=OneMinus(None=Multiply(A=@n38, B=@n39)), B=Constant() {r=0.2})
  n41 = Constant() {r=0.3}
  LinearInterpolate(
   A=StaticSwitchParameter(True=LinearInterpolate(A=@n36, B=@n37, Alpha=@n10.G), False=@n35) {defaultValue=true, parameterName=Lods_Alpha}
   B=Multiply(A=ScalarParameter() {defaultValue=1, parameterName=Roughness_Tire}, B=LinearInterpolate(A=@n40, B=@n41, Alpha=@n4.G))
   Alpha=NamedRerouteUsage()
  )
  n42 = TextureSampleParameter2D() {
   parameterName=Rim_N
   samplerSource=SSM_Wrap_WorldGroupSettings
   samplerType=SAMPLERTYPE_Normal
   texture=/VehicleCosmetics/Wheels/Visceral_Tier1/Textures/T_Visceral_Tier1_Rim_N
  }
  n43 = TextureSampleParameter2D() {
   parameterName=Tire_N
   samplerSource=SSM_Wrap_WorldGroupSettings
   samplerType=SAMPLERTYPE_Normal
   texture=&T_Visceral_Tier1_Tire_N
  }
  LinearInterpolate(A=@n42.RGB, B=@n43.RGB, Alpha=NamedRerouteUsage())
  n44 = Constant() {r=1}
  n45 = TextureSampleParameter2D() {
   parameterName=Alpha_MAsk
   samplerSource=SSM_Wrap_WorldGroupSettings
   samplerType=SAMPLERTYPE_LinearColor
   texture=/VehicleCosmetics/Wheels/Visceral_Tier1/Textures/T_Visceral_Tier1_Rim_Alpha
  }
  n46 = OneMinus(None=NamedRerouteUsage())
  Multiply(A=StaticSwitchParameter(True=LinearInterpolate(A=@n44, B=@n45.R, Alpha=@n46), False=Constant() {r=1}) {defaultValue=true, parameterName=Lods_Alpha})
  Multiply()
  NamedRerouteDeclaration(None=@n9.R) {name=VC_RC, nodeColor=<rgba 0 1 0.5313 1>}
  NamedRerouteDeclaration(None=@n4.G) {name=Tier_Mask, nodeColor=<rgba 0.3125 1 0 1>}
  OneMinus(None=Clamp(None=@n21) {maxDefault=0.05})
 ]
}
