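# Translucent unlit parent material for the holographic Visceral Tier 1 vehicle wheel, enabled for skeletal meshes.
# An offset function borrowed from the ClockGear wheel layers the ring mask, a time-based sine optionally pulses it, and the mask's green channel is added.
# The rings, tinted by brightened CustomColor, blend into a parallax radial fog built with BumpOffset.
# Static switches toggle the ring logic and painted versus fixed fog color.
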Material /VehicleCosmetics/Wheels/Visceral_Tier1/Materials/M_ParentWheel_VisceralTier1_Hologram
graph {
 properties {
  bAutomaticallySetUsageInEditor=false
  bUsedWithSkeletalMesh=true
  blendMode=BLEND_Translucent
  shadingModel=MSM_Unlit
  usedShadingModels=Unlit
 }
 nodes [
  n1 = Add(A=TextureCoordinate(), B=Constant() {r=-0.5})
  n2 = OneMinus(None=DotProduct(A=@n1, B=@n1))
  n3 = Constant() {r=0.75}
  Clamp(None=Ceil(None=Subtract(A=@n2, B=@n3)))
  Constant3Vector() {constant=<rgba 1 0.4396 0.9752 1>}
  n4 = Clamp(None=MaterialFunctionCall!(
    Texture=TextureObject() {texture=/VehicleCosmetics/Wheels/Visceral_Tier1/Textures/RingMask_M}
    Offset=ScalarParameter() {defaultValue=-1.5, parameterName=Offset}
    UVs=TextureCoordinate()
    RotateUVs=StaticBool()
    GradientFalloff
    GradientDarkness=ScalarParameter() {defaultValue=0.01, parameterName=GradientDarkness}
   ) {materialFunction=/VehicleCosmetics/Wheels/ClockGear_SE/Materials/MF_Infinity_OffsetFunction}.Result)
  n5 = Time()
  n6 = Clamp(None=Subtract(A=@n4, B=Sine(None=@n5)))
  n7 = Add(
   A=StaticSwitchParameter(True=Multiply(A=@n4, B=@n6), False=@n4) {defaultValue=true, parameterName=RingRules}
   B=TextureSample() {samplerSource=SSM_Wrap_WorldGroupSettings, texture=&RingMask_M}.G
  )
  n8 = VectorParameter() {parameterName=CustomColor, defaultValue=<rgba 1 0.2 0.4 1>}
  n9 = Multiply(A=@n8.RGBA, B=ScalarParameter() {defaultValue=4, parameterName=ColorBrightness})
  n10 = BumpOffset(Coordinate=Add(A=TextureCoordinate(), B=Constant() {r=-0.5}), Height=ScalarParameter() {defaultValue=-16, parameterName=FogOffset})
  n11 = OneMinus(None=DotProduct(A=@n10, B=@n10))
  LinearInterpolate(
   A=StaticSwitchParameter(True=Multiply(A=@n7, B=@n9), False=ScalarParameter() {parameterName=Rings}) {defaultValue=true, parameterName=RingRules}
   B=StaticSwitchParameter(True=@n8.RGBA, False=VectorParameter() {parameterName=FogColor, defaultValue=<rgba 0.02 0.51 0.51 1>}.RGBA) {defaultValue=true, parameterName=Painted}
   Alpha=Power(Base=Clamp(None=@n11), Exp=ScalarParameter() {defaultValue=4, parameterName=FogFallof})
  )
  Multiply(A=@n4, B=Constant3Vector() {constant=<rgba 1 0.1162 0.5745 1>})
 ]
}
