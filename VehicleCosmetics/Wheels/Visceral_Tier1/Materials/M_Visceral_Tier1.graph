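# Masked, lit material for the teeth of the Visceral Tier 1 vehicle wheel cosmetic, usable on skeletal meshes.
# A Painted static switch chooses CustomColor or the base texture, both AO-multiplied and partly desaturated, then a mask's green channel blends in brightened base texture.
# Vertex color RGB weights sine and cosine time waves that scale pivot-relative position by Offset, producing a pulsing vertex displacement.
# Alpha and normal textures plus Roughness and Metallic scalars complete the surface.
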Material /VehicleCosmetics/Wheels/Visceral_Tier1/Materials/M_Visceral_Tier1
graph {
 properties {bAutomaticallySetUsageInEditor=false, bUsedWithSkeletalMesh=true, blendMode=BLEND_Masked, usedShadingModels=Default Lit}
 nodes [
  n1 = TextureSample() {
   samplerSource=SSM_Wrap_WorldGroupSettings
   samplerType=SAMPLERTYPE_LinearColor
   texture=/VehicleCosmetics/Wheels/Visceral_Tier1/Textures/T_Visceral_Tier1_Teeth_AO
  }
  n2 = Multiply(A=VectorParameter() {parameterName=CustomColor, defaultValue=<rgba 1 1 1 1>}.RGB, B=@n1.RGB)
  n3 = Constant() {r=0.2}
  n4 = Desaturation(
   None=Multiply(
    A=TextureSample() {samplerSource=SSM_Wrap_WorldGroupSettings, texture=/VehicleCosmetics/Wheels/Visceral_Tier1/Textures/T_Visceral_Tier1_Teeth_BC}.RGB
    B=@n1.RGB
   )
   Fraction=@n3
  ) {luminanceFactors=<rgba 0.3 0.59 0.11 1>}
  LinearInterpolate(
   A=StaticSwitchParameter(True=Desaturation(None=@n2, Fraction=@n3) {luminanceFactors=<rgba 0.3 0.59 0.11 1>}, False=@n4) {parameterName=Painted}
   B=Multiply(A=@n4, B=Constant() {r=1.7})
   Alpha=TextureSample() {samplerSource=SSM_Wrap_WorldGroupSettings, texture=/VehicleCosmetics/Wheels/Visceral_Tier1/Textures/T_Visceral_Tier1_Teeth_M}.G
  )
  n5 = VertexColor()
  n6 = Desaturation(None=@n5) {luminanceFactors=<rgba 0.3 0.59 0.11 0>}
  n7 = Constant() {r=256}
  n8 = Subtract(
   A=WorldPosition().XYZ
   B=MaterialFunctionCall() {materialFunction=&/ObjectPivotPoint}.ObjectPivotLoacation
  )
  n9 = Time()
  n10 = Multiply(A=Sine(None=@n9) {period=0.5}, B=@n5.R)
  n11 = Multiply(A=Cosine(None=@n9), B=@n5.G)
  n12 = Sine(None=@n9)
  n13 = ConstantBiasScale(None=Add(A=Add(A=@n10, B=@n11), B=Multiply(A=@n12, B=@n5.B))) {bias=2}
  Multiply(A=Clamp(None=Multiply(A=@n6, B=@n7)), B=Multiply(A=ScalarParameter() {defaultValue=-0.075, parameterName=Offset}, B=Multiply(A=@n8, B=@n13)))
  ScalarParameter() {defaultValue=0.6, parameterName=Roughness}
  ScalarParameter() {parameterName=Metallic}
  TextureSampleParameter2D() {
   parameterName=Alpha
   samplerSource=SSM_Wrap_WorldGroupSettings
   samplerType=SAMPLERTYPE_LinearColor
   texture=/VehicleCosmetics/Wheels/Visceral_Tier1/Textures/T_Visceral_Tier1_Teeth_Alpha
  }
  TextureSample() {
   samplerSource=SSM_Wrap_WorldGroupSettings
   samplerType=SAMPLERTYPE_Normal
   texture=/VehicleCosmetics/Wheels/Visceral_Tier1/Textures/T_Visceral_Tier1_Teeth_N
  }
 ]
}
